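# Niagara skydiving contrail cosmetic "Recurve", with purple trail and glow, flares, sparks, spinning icon meshes and pale-blue camera-facing flourishes.
# FN_Contrail_Setup supplies a velocity direction that aims trail meshes, flares, sparks and icons, which sample skeletal mesh locations.
# The Camera emitter instead rings meshes and sprites around the system, timing a mid-life ray flash.
# User.ParaGlideLeanAlpha blends spawn rates, speeds, trail stretch, orientation and material intensity.
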NiagaraSystem /BRCosmetics/Athena/Effects/Particles/Skydiving/NS_Contrail_Recurve/NS_Contrail_Recurve
dependencies {
 usedDataInterfaces [
  &NiagaraDataInterfaceArrayFloat
  &NiagaraDataInterfaceMeshRendererInfo
  &NiagaraDataInterfaceSpriteRendererInfo
  &NiagaraDataInterfaceCurve
  &NiagaraDataInterfaceSkeletalMesh
 ]
 usedDynamicInputs [
  &Frac_Float
  &ReturnNormalizedExecIndex
  &MakeVector2D
  &MakeFloatFromVector
  /Niagara/DynamicInputs/CascadeConversion/SplitVector
  &FN_FrontEndFloatModifier
  &Lerp_Vector2
 ]
 usedModules [
  &self:SystemSpawnScript.SystemScriptSource.SystemScriptGraph.NiagaraNodeAssignment_0.SetVariables_#1
  &SystemState
  /Game/Effects/Niagara/Modules/Contrails/FN_Contrail_Setup
  &FN_Integrator
 ]
}
emitters {
 Camera {
  emitterUpdateScript [
   &EmitterState(
    Life Cycle Mode=::NewEnumerator1
    Loop Duration=&RandomRangeFloat(Minimum=1, Maximum=1.5)
    UseLoopDelay=false
   )
   &SpawnRate(SpawnRate=&Lerp_Float(A=1, B=2.5, Alpha=$User.ParaGlideLeanAlpha), Use Spawn Probability=false)
  ]
  particleSpawnScript [
   &InitializeParticle(
    Color=<rgba 0.5677 0.8668 1 1>
    Lifetime=&RandomRangeFloat(Minimum=1.5, Maximum=2.5)
    Mesh Renderer Array Visibility Mode=::NewEnumerator1
    Mesh Renderer Info=DataInterface() {propertyValues {MeshRenderer=&self:Camera_0.NiagaraMeshRendererProperties_0}}
    Mesh Renderer Mesh Index:NiagaraInt32=&RandomRangeInt(Minimum=0, Maximum=2)
    Mesh Scale Mode=::NewEnumerator3
    Mesh Uniform Scale=&RandomRangeFloat(Minimum=0.09, Maximum=0.12)
    Sprite Rotation Mode=::NewEnumerator2
    Sprite Size Mode=::NewEnumerator3
    Uniform Sprite Size=150
    UsePositionOffset=false
    Write Material Random=true
   )
   /SetVariables(
    Particles.Radius:NiagaraFloat=&RandomRangeFloat(Minimum=140, Maximum=200)
    Particles.Seed:NiagaraFloat=&Add_Float(
     A=&Divide_Float(
      A=&MakeFloatFromInt(Int=&ReturnParticleIDAttributes())
      B=&MakeFloatFromInt(Int=&RandomRangeInt(Minimum=3, Maximum=5))
     )
     B=&RandomRangeFloat(Minimum=-0.05, Maximum=0.05)
    )
   ) {isSetParameters=true}
   &SystemLocation(
    Offset=&MakeVector(
     X=&Sine(Normalized Angle=$Particles.Seed, Period=1, Scale=$Particles.Radius, Bias=0)
     Y=&Cosine(Normalized Angle=$Particles.Seed, Period=1, Scale=$Particles.Radius, Bias=0)
     Z=&RandomRangeFloat(Minimum=-150, Maximum=-250)
    )
    Offset Coordinate Space=::Simulation
   )
   &SubUVAnimation(
    Evaluation Type=::NewEnumerator0
    "Sprite Renderer "=DataInterface() {propertyValues {SpriteRenderer=&self:Camera_0.NiagaraSpriteRendererProperties_0}}
    SubUV Animation Mode:ENiagara_SubUVLookupModeV2=::NewEnumerator2
    UseEndFrame=false
    UseStartFrame=false
   )
   &AddVelocity(
    Cone Angle=10
    Cone Axis=<0 0 1>
    Module.UseConeFalloff=false
    Module.UseVelDistribution=false
    Override Seed=false
    Velocity Mode=::NewEnumerator2
    Velocity Speed=&Lerp_Float(
     A=&RandomRangeFloat(Minimum=250, Maximum=400)
     B=&RandomRangeFloat(Minimum=350, Maximum=500)
     Alpha=$User.ParaGlideLeanAlpha
    )
   )
   /SetVariables(Particles.CameraOffsetFlare=15) {isSetParameters=true}
   &InitialMeshOrientation(
    Mesh Orientation Mode=::NewEnumerator7
    Rotation=&MakeVector(X=&RandomRangeFloat(Minimum=-0.1, Maximum=0.1), Y=0, Z=0)
    Use Rotation Vector=true
   )
  ]
  particleUpdateScript [
   &ParticleState()
   &ScaleSpriteSize(
    Scale Sprite Size Mode=::NewEnumerator0
    Uniform Scale Factor=&FloatFromCurve(FloatCurve=DataInterface() {propertyValues {Curve {
        defaultValue=1
        keys [
         {interpMode=RCIM_Cubic, tangentMode=RCTM_User, time=0.45, value=0}
         {arriveTangentWeight=0.117, interpMode=RCIM_Cubic, leaveTangentWeight=0.0582, tangentMode=RCTM_User, time=0.475, value=1}
         {interpMode=RCIM_Cubic, tangentMode=RCTM_User, time=0.5, value=0.5}
         {arriveTangentWeight=0.0577, interpMode=RCIM_Cubic, leaveTangentWeight=0.1176, tangentMode=RCTM_User, time=0.52, value=1}
         {tangentMode=RCTM_User, time=0.55, value=0}
        ]
       }}})
   )
   &ScaleMeshSize(Scale Factor=&VectorFromFloat(Value=&FloatFromCurve(FloatCurve=DataInterface() {propertyValues {Curve {
         defaultValue=0
         keys [
          {interpMode=RCIM_Cubic, leaveTangentWeight=0.0406, tangentMode=RCTM_User, time=0, value=0}
          {arriveTangentWeight=0.177, interpMode=RCIM_Cubic, leaveTangentWeight=0.1654, tangentMode=RCTM_User, time=0.1217, value=1.5904}
          {arriveTangentWeight=0.045, interpMode=RCIM_Cubic, leaveTangentWeight=0.1985, tangentMode=RCTM_User, time=0.2354, value=1}
          {arriveTangentWeight=0.1963, interpMode=RCIM_Cubic, leaveTangentWeight=0.0488, tangentMode=RCTM_User, time=0.6, value=1}
          {arriveTangentWeight=0.1638, interpMode=RCIM_Cubic, leaveTangentWeight=0.1809, tangentMode=RCTM_User, time=0.7468, value=1.1}
          {arriveTangentWeight=0.046, interpMode=RCIM_Cubic, tangentMode=RCTM_User, time=1, value=0}
         ]
        }}})))
   &ScaleVelocity(Velocity Scale=&VectorFromFloat(Value=&FloatFromCurve(FloatCurve=DataInterface() {propertyValues {Curve {keys [
          {interpMode=RCIM_Cubic, leaveTangentWeight=0.0543, tangentMode=RCTM_User, time=0.3201, value=1}
          {
           arriveTangentWeight=0.0385
           interpMode=RCIM_Cubic
           leaveTangentWeight=0.0298
           tangentMode=RCTM_Break
           tangentWeightMode=RCTWM_WeightedBoth
           time=0.5276
           value=0.25
          }
          {
           arriveTangent=1.5207
           arriveTangentWeight=0.1108
           interpMode=RCIM_Cubic
           leaveTangent=1.5207
           leaveTangentWeight=0.2353
           tangentMode=RCTM_User
           time=0.5999
           value=0.3765
          }
          {arriveTangentWeight=0.0534, interpMode=RCIM_Cubic, tangentMode=RCTM_User, time=0.6753, value=1.7}
         ]}}})))
   &SolveForcesAndVelocity(Clamp Velocity=false, Limit Acceleration=false)
   /SetVariables(
    Particles.RayColor:LinearColor=&MakeLinearColorFromVectorAndFloat("Vector (RGB)"=<0.5 1.5 2>, "Float (Alpha)"=&RandomRangeFloat(Minimum=0.05, Maximum=0.07))
    Particles.RaySize:Vector2f=&Vector2DFromFloat(Value=&FloatFromCurve(
      FloatCurve=DataInterface() {propertyValues {Curve {
         defaultValue=1
         keys [
          {interpMode=RCIM_Cubic, tangentMode=RCTM_User, time=0.45, value=0}
          {arriveTangentWeight=0.117, interpMode=RCIM_Cubic, leaveTangentWeight=0.0582, tangentMode=RCTM_User, time=0.475, value=1}
          {interpMode=RCIM_Cubic, tangentMode=RCTM_User, time=0.5, value=0.5}
          {arriveTangentWeight=0.0577, interpMode=RCIM_Cubic, leaveTangentWeight=0.1176, tangentMode=RCTM_User, time=0.52, value=1}
          {tangentMode=RCTM_User, time=0.55, value=0}
         ]
        }}}
      Scale Curve=400
     ))
   ) {isSetParameters=true}
   &UpdateMeshOrientation(
    Facing Interpolation=true
    "Facing Origin ":NiagaraPosition=&AddVectorToPosition(Position=$Particles.Position, Vector=&MakeVector(X=0, Y=0, Z=130))
    Facing Position:NiagaraPosition=&AddVectorToPosition(Position=$Engine.Emitter.SimulationPosition, Vector=<0 0 50>)
    Orientation Method=::NewEnumerator2
    Override Quat=false
    TransformFacingOrigin=false
    TransformFacingPosition=false
    TransformUp=false
   )
   &UpdateMeshOrientation(Rotation Rate=&RandomRangeFloat(Minimum=-1, Maximum=1), Rotation Vector=<0.1 0.02 0.02>) {module=UpdateMeshOrientation001}
   &UpdateMeshOrientation(
    Rotation Rate=&FloatFromCurve(FloatCurve=DataInterface() {propertyValues {Curve {
        defaultValue=1
        keys [{interpMode=RCIM_Cubic, tangentMode=RCTM_User, time=0.6021, value=0}, {interpMode=RCIM_Cubic, tangentMode=RCTM_User, time=1, value=1}]
       }}})
    Rotation Vector=&RandomVector(Vector Scale=3)
   ) {module=UpdateMeshOrientation002}
   &AlignSpriteToMeshOrientation(Mesh Orientation Relative Sprite Alignment Vector=<1 0 0>, SetAlignment=true, SetFacing=true)
   /SetVariables() {isSetParameters=true}
  ]
  rendererClasses [&NiagaraMeshRendererProperties, &NiagaraSpriteRendererProperties]
  renderers [
   {rendererClass=&NiagaraMeshRendererProperties, rendererIndex=0}
   {rendererClass=&NiagaraSpriteRendererProperties, rendererIndex=1}
   {rendererClass=&NiagaraSpriteRendererProperties, rendererIndex=2}
  ]
 }
 Flares {
  emitterUpdateScript [
   &EmitterState(Life Cycle Mode=::NewEnumerator1, Loop Behavior=::NewEnumerator1, Loop Duration Mode=::NewEnumerator1, UseLoopDelay=false)
   &SpawnRate(SpawnRate=10, Use Spawn Probability=false)
  ]
  particleSpawnScript [
   &InitializeParticle(
    Color=<rgba 2 0.6 3 1>
    Lifetime Max=0.8
    Lifetime Min=0.6
    Lifetime Mode=::NewEnumerator1
    Sprite Size Mode=::NewEnumerator1
    Uniform Sprite Size Max=60
    Uniform Sprite Size Min=40
    UsePositionOffset=false
    Write Material Random=true
   )
   &AddVelocity(
    Cone Angle=15
    Cone Axis=$System.FN_Contrail_Setup.Direction Velocity
    Module.UseConeFalloff=false
    Module.UseVelDistribution=false
    Override Seed=false
    Velocity Mode=::NewEnumerator2
    Velocity Speed=&RandomRangeFloat(Minimum=0, Maximum=30)
   )
  ]
  particleUpdateScript [
   &ParticleState()
   &SkeletalMeshLocation(
    Bone / Socket Index=&ModuloInt(A=&ReturnParticleIDAttributes(), B=2)
    MaskSpawnGroup=false
    Mesh Sampling Type=::NewEnumerator5
    Position Sampling=::NewEnumerator1
    Sampled Position Offset=&RandomVector(Vector Scale=3)
    Skeletal Mesh=$System.Skeletal Mesh
    Socket Sampling Mode=::NewEnumerator5
   )
   &ScaleSpriteSize(Uniform Curve Sprite Scale=DataInterface() {propertyValues {Curve {
       defaultValue=0
       keys [
        <at 0 0 RCIM_Cubic>
        {
         arriveTangent=0.3487
         arriveTangentWeight=0.0221
         interpMode=RCIM_Cubic
         leaveTangent=0.3487
         leaveTangentWeight=0.0221
         tangentMode=RCTM_User
         tangentWeightMode=RCTWM_WeightedBoth
         time=0.25
         value=1
        }
        <at 1 0 RCIM_Cubic>
       ]
      }}})
   &Drag(Use Linear Drag=true, Use Rotational Drag=false)
   &SolveForcesAndVelocity(Clamp Velocity=false, Limit Acceleration=false)
   &ScaleColor(
    Scale Alpha=&FloatFromCurve(FloatCurve=DataInterface() {propertyValues {Curve {
        defaultValue=0
        keys [{interpMode=RCIM_Cubic, tangentMode=RCTM_User, time=0.5, value=1}, {interpMode=RCIM_Cubic, tangentMode=RCTM_User, time=1, value=0}]
       }}})
    ScaleA=true
    ScaleRGB=false
   )
   &CameraOffset(Camera Offset Amount=35)
  ]
  rendererClasses [&NiagaraSpriteRendererProperties]
  renderers [{rendererClass=&NiagaraSpriteRendererProperties}]
 }
 Icons {
  emitterUpdateScript [
   &EmitterState(
    Life Cycle Mode=::NewEnumerator1
    Loop Duration=&RandomRangeFloat(Minimum=0.5, Maximum=1)
    UseLoopDelay=false
   )
   &SpawnRate(
    SpawnRate=&Lerp_Float(A=&RandomRangeFloat(Minimum=8, Maximum=10), B=20, Alpha=$User.ParaGlideLeanAlpha)
    Use Spawn Probability=false
   )
  ]
  particleSpawnScript [
   &PartitionParticles()
   &InitializeParticle(
    Color Mode=::NewEnumerator0
    Lifetime=&Lerp_Float(
     A=&RandomRangeFloat(Minimum=0.75, Maximum=1)
     B=&RandomRangeFloat(Minimum=0.5, Maximum=1)
     Alpha=$User.ParaGlideLeanAlpha
    )
    Mesh Renderer Array Visibility Mode=::NewEnumerator1
    Mesh Renderer Info=DataInterface() {propertyValues {MeshRenderer=&self:Icons_0.NiagaraMeshRendererProperties_0}}
    Mesh Renderer Mesh Index:NiagaraInt32=&RandomRangeInt(Minimum=0, Maximum=1)
    Mesh Scale Mode=::NewEnumerator3
    Mesh Uniform Scale=&RandomRangeFloat(Minimum=0.05, Maximum=0.1)
    UsePositionOffset=false
    Write Material Random=true
   )
   &AddVelocity(
    Cone Angle=7
    Cone Axis=&RotateVectorByQuaternion(
     Vector To Rotate=<0 0 1>
     Quaternion=&LerpQuaternion(
      Quaternion A:Quat4f=&MakeQuaternion(
       Quaternion From=3
       Axis=<1 0 0>
       Angle=0
       From=<0 0 1>
       To=$System.FN_Contrail_Setup.Direction Velocity
       Stabilization Mode:ENiagara_OrientVectorCount=::NewEnumerator0
       Side Direction:Vector3f=?
       Up Direction:Vector3f=?
       Angle Type=::NewEnumerator0
       Normalize:NiagaraBool=true
      )
      Quaternion B:Quat4f=&MakeQuaternion(
       Quaternion From=3
       Axis=<1 0 0>
       Angle=0
       From=&MakeVector(
        X=&SelectFloatFromArray(Direct Array Index=$Particles.PartitionParticles.Partition, Float Selection Array=DataInterface() {propertyValues {FloatData [-0.06,0.06]}})
        Y=0
        Z=1
       )
       To=$System.FN_Contrail_Setup.Direction Velocity
       Stabilization Mode:ENiagara_OrientVectorCount=::NewEnumerator0
       Side Direction:Vector3f=?
       Up Direction:Vector3f=?
       Angle Type=::NewEnumerator0
       Normalize:NiagaraBool=true
      )
      Lerp Factor:NiagaraFloat=$User.ParaGlideLeanAlpha
      Enable Spherical Linear Interpolation:NiagaraBool=false
     )
    )
    Module.UseConeFalloff=false
    Module.UseVelDistribution=false
    Override Seed=false
    Velocity Mode=::NewEnumerator2
    Velocity Speed=&Lerp_Float(
     A=&RandomRangeFloat(Minimum=400, Maximum=300)
     B=&RandomRangeFloat(Minimum=700, Maximum=800)
     Alpha=$User.ParaGlideLeanAlpha
    )
   )
   &InitialMeshOrientation(
    Mesh Orientation Mode=::NewEnumerator6
    Orientation Quaternion:Quat4f=&MakeQuaternion(
     Quaternion From=0
     Axis=<0 0 1>
     Angle=&RandomRangeFloat(Minimum=0, Maximum=360)
     From=?
     To=?
     Stabilization Mode:ENiagara_OrientVectorCount=::NewEnumerator0
     Side Direction:Vector3f=?
     Up Direction:Vector3f=?
     Angle Type=::NewEnumerator0
     Coordinate Space=::World
     Normalize:NiagaraBool=true
     Rotation Matrix=Linked() {linkedVariable {name=Engine.Owner.SystemLocalToWorldNoScale, type=NiagaraFloat}}
    )
    Use Rotation Vector=false
   )
  ]
  particleUpdateScript [
   &ParticleState()
   &UpdateMeshOrientation(
    Rotation Rate=&FloatFromCurve(
     FloatCurve=DataInterface() {propertyValues {Curve {
        defaultValue=1
        keys [
         {interpMode=RCIM_Cubic, tangentMode=RCTM_User, time=0, value=0}
         {arriveTangent=6.4765, interpMode=RCIM_Cubic, leaveTangent=6.4765, tangentMode=RCTM_User, time=0.5, value=1}
        ]
       }}}
     Scale Curve=&Lerp_Float(
      A=&RandomRangeFloat(Minimum=1.5, Maximum=2)
      B=&RandomRangeFloat(Minimum=1.8, Maximum=2.3)
      Alpha=$User.ParaGlideLeanAlpha
     )
    )
    Rotation Vector=&RandomVector(Vector Scale=1)
   ) {module=UpdateMeshOrientation001}
   &ScaleMeshSize(Scale Factor=&VectorFromFloat(Value=&FloatFromCurve(FloatCurve=DataInterface() {propertyValues {Curve {
         defaultValue=0
         keys [
          {interpMode=RCIM_Cubic, leaveTangentWeight=0.0406, tangentMode=RCTM_User, time=0, value=0}
          {arriveTangentWeight=0.177, interpMode=RCIM_Cubic, leaveTangentWeight=0.1654, tangentMode=RCTM_User, time=0.15, value=1.4}
          {arriveTangentWeight=0.045, interpMode=RCIM_Cubic, leaveTangentWeight=0.1985, tangentMode=RCTM_User, time=0.2, value=1}
          {arriveTangentWeight=0.1833, interpMode=RCIM_Cubic, leaveTangentWeight=0.4187, tangentMode=RCTM_Break, time=0.8, value=1}
          {arriveTangentWeight=0.046, interpMode=RCIM_Cubic, tangentMode=RCTM_User, time=1, value=0}
         ]
        }}})))
   &SkeletalMeshLocation(
    Bone / Socket Index=$Particles.PartitionParticles.Partition
    MaskSpawnGroup=false
    Mesh Sampling Type=::NewEnumerator5
    Orientation Sampling=::NewEnumerator1
    Position Sampling=::NewEnumerator1
    Sampled Position Offset=&Add_Vector(A=&RandomVector(Vector Scale=5), B=<0 0 5>)
    Skeletal Mesh=$System.Skeletal Mesh
    Socket Sampling Mode=::NewEnumerator5
   )
   &ScaleColor(
    Scale Alpha=&FloatFromCurve(
     FloatCurve=DataInterface() {propertyValues {Curve {keys [
         <at -0.0063 1 RCIM_Cubic>
         <at 0.2782 0.6612 RCIM_Cubic>
         {arriveTangent=-0.916, interpMode=RCIM_Cubic, leaveTangent=-0.916, time=0.5256, value=1}
         <at 1 0 RCIM_Cubic>
        ]}}}
     CurveIndex=&RandomRangeFloat(Minimum=0.5, Maximum=1)
    )
    ScaleA=true
    ScaleRGB=true
   )
   &Drag(Drag=3, Use Linear Drag=true, Use Rotational Drag=true)
   &SolveForcesAndVelocity(Clamp Velocity=false, Limit Acceleration=false)
   &CameraOffset(Camera Offset Amount=15)
  ]
  rendererClasses [&NiagaraMeshRendererProperties]
  renderers [{rendererClass=&NiagaraMeshRendererProperties, rendererIndex=0}]
 }
 Sparks {
  emitterUpdateScript [
   &EmitterState(Life Cycle Mode=::NewEnumerator1, UseLoopDelay=false)
   &SpawnRate(SpawnRate=&Lerp_Float(A=100, B=100, Alpha=$User.ParaGlideLeanAlpha), Use Spawn Probability=false)
  ]
  particleSpawnScript [
   &InitializeParticle(
    Color=<rgba 673.3332 20 1000 1>
    Lifetime=&Lerp_Float(
     A=&RandomRangeFloat(Minimum=0.1, Maximum=0.4)
     B=&RandomRangeFloat(Minimum=0.1, Maximum=0.25)
     Alpha=$User.ParaGlideLeanAlpha
    )
    Sprite Size=&Multiply_Vector2D_ByFloat(Float=&RandomRangeFloat(Minimum=0.5, Maximum=2), Vector2D=<1 6>)
    Sprite Size Mode=::NewEnumerator4
    UsePositionOffset=false
    Write Material Random=true
   )
   &AddVelocity(
    Cone Angle=&Lerp_Float(A=30, B=15, Alpha=$User.ParaGlideLeanAlpha)
    Cone Axis=$System.FN_Contrail_Setup.Direction Velocity
    Module.UseConeFalloff=false
    Module.UseVelDistribution=false
    Override Seed=false
    Velocity Mode=::NewEnumerator2
    Velocity Speed=&Lerp_Float(
     A=&RandomRangeFloat(Minimum=100, Maximum=200)
     B=&RandomRangeFloat(Minimum=350, Maximum=450)
     Alpha=$User.ParaGlideLeanAlpha
    )
   ) {module=AddVelocity001}
  ]
  particleUpdateScript [
   &ParticleState()
   &SkeletalMeshLocation(
    MaskSpawnGroup=false
    Mesh Sampling Type=::NewEnumerator5
    Orientation Sampling=::NewEnumerator1
    OverrideRandomSeed=false
    Position Sampling=::NewEnumerator1
    Sampled Position Offset=&RandomVector(Vector Scale=3)
    Sampled Position Offset Coordinate Space:ENiagaraSkelMeshTransforms=::NewEnumerator2
    Skeletal Mesh=$System.Skeletal Mesh
   )
   &ScaleSpriteSize(Uniform Curve Sprite Scale=DataInterface() {propertyValues {Curve {
       defaultValue=0
       keys [
        {interpMode=RCIM_Cubic, tangentMode=RCTM_User, time=0, value=1}
        {arriveTangent=-3.1025, interpMode=RCIM_Cubic, leaveTangent=-3.1025, tangentMode=RCTM_User, time=1, value=0}
       ]
      }}})
   &SolveForcesAndVelocity(Clamp Velocity=false, Limit Acceleration=false)
  ]
  rendererClasses [&NiagaraSpriteRendererProperties]
  renderers [{rendererClass=&NiagaraSpriteRendererProperties}]
 }
 TrailAndHandsGlow {
  emitterUpdateScript [
   &EmitterState(Life Cycle Mode=::NewEnumerator1, Loop Behavior=::NewEnumerator1, Loop Duration Mode=::NewEnumerator1, UseLoopDelay=false)
   &SpawnBurst_Instantaneous(Spawn Count=2, Spawn Group=1, Use Spawn Probability=false, UseLoopCountLimit=false) {module=SpawnBurst_Instantaneous001}
   &SpawnBurst_Instantaneous(Spawn Count=4, Use Spawn Probability=false, UseLoopCountLimit=false)
  ]
  particleSpawnScript [
   &PartitionParticles(Normalized Index Includes One=true)
   &InitializeParticle(
    Color=<rgba 0.8207 0.2234 1 0.5>
    Lifetime=5
    Mesh Scale Mode=::NewEnumerator3
    Sprite Size Mode=::NewEnumerator3
    Uniform Sprite Size=100
    UsePositionOffset=false
    Write Material Random=true
   )
   /SetVariables(
    Particles.CameraOffsetGlow=25
    Particles.DynamicMaterialParametergGlow:Vector4f=&MakeVector4(X=8, Y=0, Z=0, W=0)
    Particles.GlowColor=&FNTimeOfDayColor(
     Night Color=<rgba 2.9983 0.14 7 0.2>
     Morning Color=<rgba 4.7133 0.14 7 0.2>
     Day Color=<rgba 4.7133 0.14 7 0.2>
     Evening Color=<rgba 4.7133 0.14 7 0.2>
    )
    Particles.VisibilityTag=&SpawnGroup()
   ) {isSetParameters=true}
  ]
  particleUpdateScript [
   &ParticleState(
    Kill Particles When Lifetime Has Elapsed=false
    Let Infinitely Lived Particles Die When Emitter Deactivates=true
    Loop Particles Lifetime=true
   )
   &SkeletalMeshLocation(
    Bone / Socket Index=$Particles.PartitionParticles.Partition
    MaskSpawnGroup=false
    Mesh Sampling Type=::NewEnumerator5
    Orientation Sampling=::NewEnumerator1
    Sampled Position Offset Coordinate Space:ENiagaraSkelMeshTransforms=::NewEnumerator2
    Skeletal Mesh=$System.Skeletal Mesh
    Socket Sampling Mode=::NewEnumerator5
   )
   &ScaleMeshSize(Scale Factor=&Lerp_Vector(A=<1 1 8>, B=<1 1 15>, Alpha=$User.ParaGlideLeanAlpha))
   &DynamicMaterialParameters(
    Index 0 Param 1=&RandomRangeFloat(Minimum=0.5, Maximum=1)
    Index 0 Param 2=&MakeCustomFloatFromBool(
     Bool=&SetBoolByFloatComparison(A=$User.ParaGlideLeanAlpha, B=0.2)
     True Float=&RandomRangeFloat(Minimum=0.8, Maximum=1)
     False Float=&RandomRangeFloat(Minimum=0.3, Maximum=0.4)
    )
    Index 3 Param 1=&Lerp_Float(A=55, B=65, Alpha=$User.ParaGlideLeanAlpha)
    Index 3 Param 2=14
    Index 3 Param 3=&MakeCustomFloatFromBool(Bool=&SetBoolByFloatComparison(A=$User.ParaGlideLeanAlpha, B=0.2), True Float=3, False Float=1.5)
    Index 3 Param 4=&MakeCustomFloatFromBool(
     Bool=&SetBoolByFloatComparison(A=$User.ParaGlideLeanAlpha, B=0.2)
     True Float=&RandomRangeFloat(Minimum=0.9, Maximum=1)
     False Float=&RandomRangeFloat(Minimum=0.7, Maximum=0.8)
    )
    Module.Index0Float0Write=true
    Module.Index0Float1Write=true
    Module.Index0Float2Write=false
    Module.Index0Float3Write=false
    Module.Index3Float0Write=true
    Module.Index3Float1Write=true
    Module.Index3Float2Write=true
    Module.Index3Float3Write=true
    Write Parameter Index 3=::NewEnumerator1
   )
   &ScaleColor(Scale Alpha=&Lerp_Float(A=0.6, B=0.7, Alpha=$User.ParaGlideLeanAlpha), ScaleA=true, ScaleRGB=true)
   /SetVariables(Particles.MeshOrientation=&LerpQuaternion(
     Quaternion A:Quat4f=&MakeQuaternion(
      Quaternion From=3
      Axis=<1 0 0>
      Angle=0
      From=<0 0 1>
      To=$System.FN_Contrail_Setup.Direction Velocity
      Stabilization Mode:ENiagara_OrientVectorCount=::NewEnumerator0
      Side Direction:Vector3f=?
      Up Direction:Vector3f=?
      Angle Type=::NewEnumerator0
      Normalize:NiagaraBool=true
     )
     Quaternion B:Quat4f=&MakeQuaternion(
      Quaternion From=3
      Axis=<1 0 0>
      Angle=0
      From=&MakeVector(
       X=&SelectFloatFromArray(Direct Array Index=$Particles.PartitionParticles.Partition, Float Selection Array=DataInterface() {propertyValues {FloatData [-0.06,0.06]}})
       Y=0
       Z=1
      )
      To=$System.FN_Contrail_Setup.Direction Velocity
      Stabilization Mode:ENiagara_OrientVectorCount=::NewEnumerator0
      Side Direction:Vector3f=?
      Up Direction:Vector3f=?
      Angle Type=::NewEnumerator0
      Normalize:NiagaraBool=true
     )
     Lerp Factor:NiagaraFloat=$User.ParaGlideLeanAlpha
     Enable Spherical Linear Interpolation:NiagaraBool=false
    )) {isSetParameters=true}
  ]
  rendererClasses [&NiagaraMeshRendererProperties, &NiagaraSpriteRendererProperties]
  renderers [
   {rendererClass=&NiagaraMeshRendererProperties, rendererIndex=0}
   {rendererClass=&NiagaraMeshRendererProperties, rendererIndex=1}
   {rendererClass=&NiagaraSpriteRendererProperties, rendererIndex=2}
  ]
 }
}
summary {
 systemName=NS_Contrail_Recurve
 emitters [
  {
   emitterName=TrailAndHandsGlow
   rendererClasses [&NiagaraMeshRendererProperties, &NiagaraSpriteRendererProperties]
  }
  {emitterName=Camera, rendererClasses [&NiagaraMeshRendererProperties, &NiagaraSpriteRendererProperties]}
  {emitterName=Flares, rendererClasses [&NiagaraSpriteRendererProperties]}
  {emitterName=Sparks, rendererClasses [&NiagaraSpriteRendererProperties]}
  {emitterName=Icons, rendererClasses [&NiagaraMeshRendererProperties]}
 ]
 userVariables [
  User.bIsFrontEnd {type=NiagaraBool}
  User.bIsFrontEndPreview {type=NiagaraBool}
  User.ParaGlideLeanAlpha {type=NiagaraFloat, defaultValue=1, description=""}
 ]
}
